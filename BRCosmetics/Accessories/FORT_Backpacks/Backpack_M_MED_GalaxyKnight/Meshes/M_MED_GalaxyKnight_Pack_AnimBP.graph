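# Anim Blueprint for the Galaxy Knight back bling that makes the pack and its six petals float and drift.
# EventGraph builds time-based sine offsets for the pack and phase-shifted ones per petal.
# AnimGraph runs spring-driven AnimDynamics on petals and pack, then layers those offsets through Modify Bone nodes.
# Petal physics weight eases down while jumping or falling, and ambient petal drift fades as movement speed rises.
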AnimBlueprint /BRCosmetics/Accessories/FORT_Backpacks/Backpack_M_MED_GalaxyKnight/Meshes/M_MED_GalaxyKnight_Pack_AnimBP
target_skeleton=&/M_MED_GalaxyKnight_Pack_Skeleton
graphs [
 AnimGraph {nodes [
   n1 = Get AmbientAlpha() -> (AmbientAlpha:double)
   n2 = Get AlphaJump() -> (AlphaJump:double)
   n3 = Anim Dynamics - Bone: Dyn_Petal_01(ComponentPose=Local To Component(LocalPose=Input Pose()), Alpha=@n2, ExternalForce="0,0,0") {
    angularBiasOverride=0.99
    angularDampingOverride=0.99
    angularSpringConstant=10
    bLinearSpring=true
    bOverrideLinearDamping=true
    bUseGravityOverride=true
    gravityScale=0.15
    linearDampingOverride=0.95
    linearSpringConstant=10
    alphaBoolBlend {blendOption=Linear}
    alphaScaleBiasClamp {bInterpResult=true, interpSpeedDecreasing=5, interpSpeedIncreasing=5}
    boundBone {boneName=Dyn_Petal_01}
    componentAppliedLinearAccClamp=<40 40 40>
    componentLinearAccScale=<0.2 0.2 0.1>
    componentLinearVelScale=<0.2 0.2 0.1>
    physicsBodyDefinitions [{
      boundBone {boneName=Dyn_Petal_01}
      boxExtents=<3 3 3>
      constraintSetup {angularLimitsMax=<0 0 0>, angularLimitsMin=<0 0 0>, angularTarget=<1 0 0>, linearAxesMax=<1 1 1>, linearAxesMin=<-1.4 -1.4 -1.4>}
      localJointOffset=<0 0 0>
     }]
   }
   n4 = Anim Dynamics - Bone: Dyn_Petal_02(ComponentPose=@n3, Alpha=@n2, ExternalForce="0,0,0") {
    angularBiasOverride=0.99
    angularDampingOverride=0.99
    angularSpringConstant=10
    bLinearSpring=true
    bOverrideLinearDamping=true
    bUseGravityOverride=true
    gravityScale=0.15
    linearDampingOverride=0.95
    linearSpringConstant=10
    alphaBoolBlend {blendOption=Linear}
    alphaScaleBiasClamp {bInterpResult=true, interpSpeedDecreasing=5, interpSpeedIncreasing=5}
    boundBone {boneName=Dyn_Petal_02}
    componentAppliedLinearAccClamp=<40 40 40>
    componentLinearAccScale=<0.2 0.2 0.1>
    componentLinearVelScale=<0.2 0.2 0.1>
    physicsBodyDefinitions [{
      boundBone {boneName=Dyn_Petal_02}
      boxExtents=<4 4 4>
      constraintSetup {angularLimitsMax=<0 0 0>, angularLimitsMin=<0 0 0>, angularTarget=<1 0 0>, linearAxesMax=<1 1 1>, linearAxesMin=<-1.4 -1.4 -1.4>}
      localJointOffset=<0 0 0>
     }]
   }
   n5 = Anim Dynamics - Bone: Dyn_Petal_03(ComponentPose=@n4, Alpha=@n2, ExternalForce="0,0,0") {
    angularBiasOverride=0.99
    angularDampingOverride=0.99
    angularSpringConstant=10
    bLinearSpring=true
    bOverrideLinearDamping=true
    bUseGravityOverride=true
    gravityScale=0.15
    linearDampingOverride=0.95
    linearSpringConstant=14
    alphaBoolBlend {blendOption=Linear}
    alphaScaleBiasClamp {bInterpResult=true, interpSpeedDecreasing=5, interpSpeedIncreasing=5}
    boundBone {boneName=Dyn_Petal_03}
    componentAppliedLinearAccClamp=<40 40 40>
    componentLinearAccScale=<0.2 0.2 0.1>
    componentLinearVelScale=<0.2 0.2 0.1>
    physicsBodyDefinitions [{
      boundBone {boneName=Dyn_Petal_03}
      boxExtents=<7 7 7>
      constraintSetup {angularLimitsMax=<0 0 0>, angularLimitsMin=<0 0 0>, angularTarget=<1 0 0>, linearAxesMax=<1 1 1>, linearAxesMin=<-1.4 -1.4 -1.4>}
      localJointOffset=<0 0 0>
     }]
   }
   n6 = Anim Dynamics - Bone: Dyn_Petal_04(ComponentPose=@n5, Alpha=@n2, ExternalForce="0,0,0") {
    angularBiasOverride=0.99
    angularDampingOverride=0.99
    angularSpringConstant=10
    bLinearSpring=true
    bOverrideLinearDamping=true
    bUseGravityOverride=true
    gravityScale=0.15
    linearDampingOverride=0.95
    linearSpringConstant=10
    alphaBoolBlend {blendOption=Linear}
    alphaScaleBiasClamp {bInterpResult=true, interpSpeedDecreasing=5, interpSpeedIncreasing=5}
    boundBone {boneName=Dyn_Petal_04}
    componentAppliedLinearAccClamp=<40 40 40>
    componentLinearAccScale=<0.2 0.2 0.1>
    componentLinearVelScale=<0.2 0.2 0.1>
    physicsBodyDefinitions [{
      boundBone {boneName=Dyn_Petal_04}
      boxExtents=<10 10 10>
      constraintSetup {angularLimitsMax=<0 0 0>, angularLimitsMin=<0 0 0>, angularTarget=<1 0 0>, linearAxesMax=<1 1 1>, linearAxesMin=<-1.4 -1.4 -1.4>}
      localJointOffset=<0 0 0>
     }]
   }
   n7 = Anim Dynamics - Bone: Dyn_Petal_05(ComponentPose=@n6, Alpha=@n2, ExternalForce="0,0,0") {
    angularBiasOverride=0.99
    angularDampingOverride=0.99
    angularSpringConstant=10
    bLinearSpring=true
    bOverrideLinearDamping=true
    bUseGravityOverride=true
    gravityScale=0.15
    linearDampingOverride=0.95
    linearSpringConstant=10
    alphaBoolBlend {blendOption=Linear}
    alphaScaleBiasClamp {bInterpResult=true, interpSpeedDecreasing=5, interpSpeedIncreasing=5}
    boundBone {boneName=Dyn_Petal_05}
    componentAppliedLinearAccClamp=<40 40 40>
    componentLinearAccScale=<0.2 0.2 0.1>
    componentLinearVelScale=<0.2 0.2 0.1>
    physicsBodyDefinitions [{
      boundBone {boneName=Dyn_Petal_05}
      boxExtents=<9 9 9>
      constraintSetup {angularLimitsMax=<0 0 0>, angularLimitsMin=<0 0 0>, angularTarget=<1 0 0>, linearAxesMax=<1 1 1>, linearAxesMin=<-1.4 -1.4 -1.4>}
      localJointOffset=<0 0 0>
     }]
   }
   n8 = Anim Dynamics - Bone: Dyn_Petal_06(ComponentPose=@n7, Alpha=@n2, ExternalForce="0,0,0") {
    angularBiasOverride=0.99
    angularDampingOverride=0.99
    angularSpringConstant=10
    bLinearSpring=true
    bOverrideLinearDamping=true
    bUseGravityOverride=true
    gravityScale=0.15
    linearDampingOverride=0.95
    linearSpringConstant=15
    alphaBoolBlend {blendOption=Linear}
    alphaScaleBiasClamp {bInterpResult=true, interpSpeedDecreasing=5, interpSpeedIncreasing=5}
    boundBone {boneName=Dyn_Petal_06}
    componentAppliedLinearAccClamp=<40 40 40>
    componentLinearAccScale=<0.2 0.2 0.1>
    componentLinearVelScale=<0.2 0.2 0.1>
    physicsBodyDefinitions [{
      boundBone {boneName=Dyn_Petal_06}
      boxExtents=<5 5 5>
      constraintSetup {angularLimitsMax=<0 0 0>, angularLimitsMin=<0 0 0>, angularTarget=<1 0 0>, linearAxesMax=<1 1 1>, linearAxesMin=<-1.4 -1.4 -1.4>}
      localJointOffset=<0 0 0>
     }]
   }
   n9 = Anim Dynamics - Bone: pack!(
    ComponentPose=@n8
    Alpha=0.7
    GravityOverride:Vector="0,0,50"
    ComponentLinearAccScale:Vector="0.1,0.1,0.1"
    ComponentLinearVelScale:Vector="0.1,0.1,0.1"
   ) -> (Pose) {
    alpha=0
    angularBiasOverride=1
    angularDampingOverride=0.75
    angularSpringConstant=20
    bGravityOverrideInSimSpace=true
    bLinearSpring=true
    bOverrideLinearDamping=true
    bUseGravityOverride=true
    bUsePlanarLimit=false
    bUseSphericalLimits=true
    gravityScale=0
    linearDampingOverride=0.96
    linearSpringConstant=42
    alphaBoolBlend {blendOption=Linear}
    boundBone {boneName=pack}
    componentAppliedLinearAccClamp=<80 80 80>
    gravityOverride {z=20}
    physicsBodyDefinitions [{
      sphereCollisionRadius=0
      boundBone {boneName=pack}
      boxExtents=<30 5 30>
      constraintSetup {twistAxis=AxisZ, angularLimitsMax=<0 0 0>, angularLimitsMin=<0 0 0>, angularTarget=<0 0 0>, linearAxesMax=<3 0 3>, linearAxesMin=<-3 -3 -3>}
      localJointOffset=<0 0 0>
     }]
    relativeSpaceBone {boneName=root}
    sphericalLimits [{limitRadius=7, limitType=Inner, drivingBone {boneName=pack}, sphereLocalOffset=<0 0 4.4>}]
   }
   n10 = "Transform (Modify) Bone - Bone: pack"!(ComponentPose=@n9, bAlphaBoolEnabled:bool, Translation=Get Backpack_floating() -> (Backpack_floating:Vector), Rotation="0,0,0", Scale) -> (Pose) {
    alphaInputType=Bool
    rotationMode=BMM_Replace
    rotationSpace=BCS_BoneSpace
    translationMode=BMM_Additive
    translationSpace=BCS_BoneSpace
    alphaBoolBlend {blendInTime=0.2, blendOption=Linear, blendOutTime=0.2}
    boneToModify {boneName=pack}
    translation {y=-3}
   }
   n11 = "Transform (Modify) Bone - Bone: pack"!(ComponentPose=@n10, Alpha:float, Translation="0,-1,0", Rotation="0,0,0", Scale) -> (Pose) {
    rotationSpace=BCS_BoneSpace
    translationMode=BMM_Additive
    translationSpace=BCS_BoneSpace
    alphaBoolBlend {blendOption=Linear}
    alphaScaleBiasClamp {interpSpeedDecreasing=4, interpSpeedIncreasing=4}
    boneToModify {boneName=pack}
    translation {y=-1}
   }
   n12 = "Transform (Modify) Bone - Bone: Dyn_Petal_01"(ComponentPose=@n11, Alpha=@n1, Translation=Get Dyn Transform_01() -> (Dyn Transform_01:Vector), Rotation="0,0,0", Scale="1,1,1") {
    rotationMode=BMM_Additive
    rotationSpace=BCS_BoneSpace
    translationMode=BMM_Additive
    alphaBoolBlend {blendOption=Linear}
    alphaScaleBiasClamp {bInterpResult=true, interpSpeedDecreasing=4, interpSpeedIncreasing=4}
    boneToModify {boneName=Dyn_Petal_01}
   }
   n13 = "Transform (Modify) Bone - Bone: Dyn_Petal_02"(ComponentPose=@n12, Alpha=@n1, Translation=Get Dyn Transform_02() -> (Dyn Transform_02:Vector), Rotation="0,0,0", Scale="1,1,1") {
    rotationMode=BMM_Additive
    rotationSpace=BCS_BoneSpace
    translationMode=BMM_Additive
    alphaBoolBlend {blendOption=Linear}
    alphaScaleBiasClamp {bInterpResult=true, interpSpeedDecreasing=4, interpSpeedIncreasing=4}
    boneToModify {boneName=Dyn_Petal_02}
   }
   n14 = "Transform (Modify) Bone - Bone: Dyn_Petal_03"(ComponentPose=@n13, Alpha=@n1, Translation=Get Dyn Transform_03() -> (Dyn Transform_03:Vector), Rotation="0,0,0", Scale="1,1,1") {
    rotationMode=BMM_Additive
    rotationSpace=BCS_BoneSpace
    translationMode=BMM_Additive
    alphaBoolBlend {blendOption=Linear}
    alphaScaleBiasClamp {bInterpResult=true, interpSpeedDecreasing=4, interpSpeedIncreasing=4}
    boneToModify {boneName=Dyn_Petal_03}
   }
   n15 = "Transform (Modify) Bone - Bone: Dyn_Petal_04"(ComponentPose=@n14, Alpha=@n1, Translation=Get Dyn Transform_04() -> (Dyn Transform_04:Vector), Rotation="0,0,0", Scale="1,1,1") {
    rotationMode=BMM_Additive
    rotationSpace=BCS_BoneSpace
    translationMode=BMM_Additive
    alphaBoolBlend {blendOption=Linear}
    alphaScaleBiasClamp {bInterpResult=true, interpSpeedDecreasing=4, interpSpeedIncreasing=4}
    boneToModify {boneName=Dyn_Petal_04}
   }
   n16 = "Transform (Modify) Bone - Bone: Dyn_Petal_05"(ComponentPose=@n15, Alpha=@n1, Translation=Get Dyn Transform_05() -> (Dyn Transform_05:Vector), Rotation="0,0,0", Scale="1,1,1") {
    rotationMode=BMM_Additive
    rotationSpace=BCS_BoneSpace
    translationMode=BMM_Additive
    alphaBoolBlend {blendOption=Linear}
    alphaScaleBiasClamp {bInterpResult=true, interpSpeedDecreasing=4, interpSpeedIncreasing=4}
    boneToModify {boneName=Dyn_Petal_05}
   }
   n17 = "Transform (Modify) Bone - Bone: Dyn_Petal_06"(ComponentPose=@n16, Alpha=@n1, Translation=Get Dyn Transform_06() -> (Dyn Transform_06:Vector), Rotation="0,0,0", Scale="1,1,1") {
    rotationMode=BMM_Additive
    rotationSpace=BCS_BoneSpace
    translationMode=BMM_Additive
    alphaBoolBlend {blendOption=Linear}
    alphaScaleBiasClamp {bInterpResult=true, interpSpeedDecreasing=4, interpSpeedIncreasing=4}
    boneToModify {boneName=Dyn_Petal_06}
   }
   Output Pose(Result=Component To Local(ComponentPose=@n17)) {layerGroup=None}
  ]}
 EventGraph {nodes [
   n1 = Event BlueprintUpdateAnimation()
   > Set Delta Time X(Delta Time X=@n1.DeltaTimeX)
   n2 = float * float(A=GetGameTimeInSeconds(), B=1.2) {is=title K2Node_PromotableOperator}
   n3 = "Sin (Radians)"(A=float * float(A=0.56, B=@n2) {is=title K2Node_CommutativeAssociativeBinaryOperator})
   n4 = "Sin (Radians)"(A=float * float(A=0.8, B=@n2))
   n5 = "Sin (Radians)"(A=float * float(A=1.21, B=@n2))
   > Set Backpack_floating(Backpack_floating:Vector=MakeVector(X=float * float(A=@n3, B=0.5), Y=float * float(A=@n5, B=0.15), Z=float * float(A=@n4, B=1))) -> (Output_Get:Vector)
   > n6 = Sequence!(execute) -> (then_0, then_1, then_2:Exec)
   n7 = OR Boolean(A=Get bIsJumping(), B=Get bIsFalling())
   Set AlphaJump(
    execute=@n6.then_2
    AlphaJump:double=FInterpTo(Current=Get AlphaJump(), Target=SelectFloat(A=0.2, B=0.7, bPickA=@n7), DeltaTime=Get Delta Time X(), InterpSpeed=5)
   ) -> (Output_Get:double)
   n8 = Vector Length(A=Get PawnVelocity())
   n9 = "To Float (Boolean)"(InBool=Get bIsMoving2D())
   Set AmbientAlpha(
    execute=@n6.then_1
    AmbientAlpha:double=MapRangeClamped(Value=float * float(A=@n8, B=@n9) {is=title K2Node_PromotableOperator}, InRangeA=10, InRangeB=550, OutRangeA=1, OutRangeB=0.3)
   ) -> (Output_Get:double)
   n10 = float * float(A=GetGameTimeInSeconds(), B=2)
   n11 = float + float(A="Sin (Radians)"(A=float + float(A=@n10, B=1)), B=1)
   Set Dyn Transform_01(
    execute=@n6.then_0
    Dyn Transform_01:Vector="Lerp (Vector)"!(A:Vector="0,-1.12,-0.56", B:Vector="0,1.12,0.56", Alpha:float=float / float(A=@n11, B=3)) -> (ReturnValue:Vector)
   ) -> (Output_Get:Vector)
   n12 = float + float(A="Sin (Radians)"(A=float + float(A=@n10, B=3)), B=1)
   > Set Dyn Transform_02(Dyn Transform_02:Vector="Lerp (Vector)"(A="0,-1.12,-0.56", B="0,1.12,0.56", Alpha=float / float(A=@n12, B=3))) -> (Output_Get:Vector)
   n13 = float + float(A="Sin (Radians)"(A=float + float(A=@n10, B=4)), B=1)
   > Set Dyn Transform_03(Dyn Transform_03:Vector="Lerp (Vector)"(A="0,-1.12,-0.56", B="0,1.12,0.56", Alpha=float / float(A=@n13, B=3))) -> (Output_Get:Vector)
   n14 = float + float(A="Sin (Radians)"(A=float + float(A=@n10, B=6)), B=1)
   > Set Dyn Transform_04(Dyn Transform_04:Vector="Lerp (Vector)"(A="0,-1.12,-0.56", B="0,1.12,0.56", Alpha=float / float(A=@n14, B=3.2))) -> (Output_Get:Vector)
   n15 = float + float(A="Sin (Radians)"(A=float + float(A=@n10, B=9)), B=1)
   > Set Dyn Transform_05(Dyn Transform_05:Vector="Lerp (Vector)"(A="0,-1.12,-0.56", B="0,1.12,0.56", Alpha=float / float(A=@n15, B=3))) -> (Output_Get:Vector)
   n16 = float + float(A="Sin (Radians)"(A=float + float(A=@n10, B=11)), B=1)
   > Set Dyn Transform_06(Dyn Transform_06:Vector="Lerp (Vector)"(A="0,-1.12,-0.56", B="0,1.12,0.56", Alpha=float / float(A=@n16, B=2.8))) -> (Output_Get:Vector)
  ]}
]
variables [
 AlphaJump {type=number, default=0.7}
 AmbientAlpha {type=number}
 Backpack_floating {type=Vector}
 Delta Time X {type=number}
 Dyn Transform_01 {type=Vector}
 Dyn Transform_02 {type=Vector}
 Dyn Transform_03 {type=Vector}
 Dyn Transform_04 {type=Vector}
 Dyn Transform_05 {type=Vector}
 Dyn Transform_06 {type=Vector}
]
